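# Translucent deferred decal material for Hera terrain, with parameterized diffuse, normal, specular-mask and alpha textures.
# An "Is WS" static switch chooses dithered, wavy-distorted triplanar world-space projection or scaled mesh UVs for color, normal and masks.
# Color gets a tint, the normal is transformed from world to tangent space and flattened, and specular and roughness come from scaled mask channels.
# Opacity scales either diffuse alpha, the default, or a separate mask texture.
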Material /Hera/Terrain/Decals/Materials/M_Hera_Indestructible_Decal_A
graph {
 properties {
  bAutomaticallySetUsageInEditor=false
  blendMode=BLEND_Translucent
  materialDomain=MD_DeferredDecal
  previewMesh=/Hera/Sets/SeaMist/Meshes/SM_SeaMist_Lighthouse_FresnelLens_A
  usedShadingModels=Default Lit
 }
 nodes [
  n1 = ScalarParameter() {defaultValue=900, parameterName=WS tiling}
  n2 = MaterialFunctionCall!(Absolute World Position, "Normal (WS)", Texture Scale=@n1, Triplanar Fade, Derivatives Scale) {materialFunction=/BRCoreMaterials/Materials/MaterialFunctions/MF_TriplanarDithered_WavyDistortion}
  n3 = TextureObjectParameter() {group=Textures, parameterName=Normal, samplerType=SAMPLERTYPE_Normal, texture=/Hermes/Sets/BigSplat/Textures/Decals/BigSplat_BugDecal_A_N}
  n4 = TextureSample!(UVs=@n2.UVs, Tex=@n3, "DDX(UVs)"=@n2.UVs, "DDY(UVs)"=@n2.UVs, Apply View MipBias) {mipValueMode=TMVM_Derivative}
  n5 = Divide(A=TextureCoordinate(), B=ScalarParameter() {defaultValue=50, parameterName=UV tiling})
  n6 = TextureSample!(UVs=@n5, Tex=@n3, Apply View MipBias) {texture=/Hermes/Sets/BigSplat/Textures/Decals/BigSplat_BugDecal_A_D}
  MaterialFunctionCall(
   Normal=Transform(None=StaticSwitchParameter(True=@n4.RGB, False=@n6.RGB) {parameterName=Is WS}) {transformSourceType=TRANSFORMSOURCE_World, transformType=TRANSFORM_Tangent}
   Flatness=ScalarParameter() {group=Textures, parameterName=FlattenNormal}
  ) {materialFunction=&FlattenNormal}
  n7 = MaterialFunctionCall(Texture Scale=@n1) {materialFunction=&MF_TriplanarDithered_WavyDistortion}
  n8 = TextureObjectParameter() {
   group=Textures
   parameterName=Specular masks
   samplerType=SAMPLERTYPE_Masks
   texture=/Hermes/Sets/BigSplat/Textures/Decals/BigSplat_BugDecal_A_S
  }
  n9 = MaterialFunctionCall(Float3=StaticSwitchParameter(
    True=TextureSample!(UVs=@n7.UVs, Tex=@n8, "DDX(UVs)"=@n7.UVs, "DDY(UVs)"=@n7.UVs, Apply View MipBias) {mipValueMode=TMVM_Derivative}.RGB
    False=TextureSample!(UVs=@n5, Tex=@n8, Apply View MipBias) {texture=&BigSplat_BugDecal_A_D}.RGB
   ) {parameterName=Is WS}) {materialFunction=&BreakOutFloat3Components}
  Multiply(A=@n9.R, B=ScalarParameter() {defaultValue=1, group=Textures, parameterName=SpecAmount})
  n10 = TextureObjectParameter() {group=Textures, parameterName=Diffuse, texture=&BigSplat_BugDecal_A_D}
  n11 = TextureObjectParameter() {group=Textures, parameterName=Alpha, texture=/Hera/Terrain/Textures/T_Hera_Puddle_A_A_D}
  Multiply(
   A=StaticSwitchParameter(
    True=TextureSample(UVs=@n5, Tex=@n10).A
    False=TextureSample(Tex=@n11) {samplerType=SAMPLERTYPE_Masks, texture=/Hera/Terrain/Decals/Textures/T_Hera_PaintPuddle_S}.RGB
   ) {defaultValue=true, parameterName=UseDiffuseAlpha}
   B=ScalarParameter() {defaultValue=1, group=Textures, parameterName=Opacity intensity}
  )
  Multiply(A=@n9.B, B=ScalarParameter() {defaultValue=1, group=Textures, parameterName=RoughnessAmount})
  n12 = MaterialFunctionCall(Texture Scale=@n1) {materialFunction=&MF_TriplanarDithered_WavyDistortion}
  StaticSwitchParameter(
   True=Multiply(
    A=TextureSample!(UVs=@n12.UVs, Tex=@n10, "DDX(UVs)"=@n12.UVs, "DDY(UVs)"=@n12.UVs, Apply View MipBias) {mipValueMode=TMVM_Derivative}.RGB
    B=VectorParameter() {parameterName=Color overlay, defaultValue=<rgba 1 1 1 1>}.RGB
   )
   False=Multiply(
    A=TextureSample!(UVs=@n5, Tex=@n10, Apply View MipBias).RGB
    B=VectorParameter() {parameterName=Color overlay, defaultValue=<rgba 1 1 1 1>}.RGB
   )
  ) {parameterName=Is WS}
 ]
}
